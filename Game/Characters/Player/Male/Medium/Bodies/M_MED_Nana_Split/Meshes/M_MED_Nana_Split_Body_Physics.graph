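# Physics asset for the M_MED_Nana_Split male medium player body, covering secondary motion of its belly, rib and neck-rib dynamic bones.
# Four rib boxes and a belly box each pair with a tiny _Base anchor body, while two neck-rib boxes attach directly to spine_05.
# Rib joints lock twist and one swing axis for hinge-like motion; belly and neck-rib joints leave all three rotational axes unlocked.
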PhysicsAsset /Game/Characters/Player/Male/Medium/Bodies/M_MED_Nana_Split/Meshes/M_MED_Nana_Split_Body_Physics
bodies {
 dyn_bellyJiggle [{shapeType=Box, extentX=8.4098, extentY=21.9168, extentZ=8.2067, center=<15.8438 0 -0.0002>, rotation=<rot 0 -90 0>}]
 dyn_bellyJiggle_Base [{shapeType=Box, extentX=1.01, extentY=1.01, extentZ=1.01, rotation=<rot 0 -90 0>}]
 dyn_neckRibA [{shapeType=Box, extentX=4.3267, extentY=12.8898, extentZ=27.7314, center=<-2.1607 0.405 0.3813>, rotation=<rot 0 -93.761 -86.4634>}]
 dyn_neckRibB [{shapeType=Box, extentX=4.7416, extentY=18.7325, extentZ=29.5239, center=<3.1512 0.0472 0.0712>, rotation=<rot 0 -93.1812 -85.2144>}]
 dyn_ribA [{shapeType=Box, extentX=5.2282, extentY=18.0942, extentZ=9.535, center=<3.9534 -0.4599 1.1635>, rotation=<rot -5.0653 -90 -17.9055>}]
 dyn_ribA_Base [{shapeType=Box, extentX=1.01, extentY=1.01, extentZ=1.01, rotation=<rot 0 -90 0>}]
 dyn_ribB [{shapeType=Box, extentX=5.4892, extentY=18.9232, extentZ=9.5367, center=<4.9211 0.1329 1.0041>, rotation=<rot -1.9917 -89.9999 -12.9388>}]
 dyn_ribB_Base [{shapeType=Box, extentX=1.01, extentY=1.01, extentZ=1.01, rotation=<rot 0 -90 0>}]
 dyn_ribC [{shapeType=Box, extentX=5.0543, extentY=19.0668, extentZ=10.1809, center=<5.4215 -0.3803 0.9223>, rotation=<rot -1.7488 -89.9998 -9.038>}]
 dyn_ribC_Base [{shapeType=Box, extentX=1.01, extentY=1.01, extentZ=1.01, rotation=<rot 0 -90 0>}]
 dyn_ribD [{shapeType=Box, extentX=9.636, extentY=18.9525, extentZ=9.9478, center=<4.9509 1.4148 0.3878>, rotation=<rot 0.4726 -89.9996 -4.3758>}]
 dyn_ribD_Base [{shapeType=Box, extentX=1.01, extentY=1.01, extentZ=1.01, rotation=<rot 0 -90 0>}]
 spine_05 [{shapeType=Box, extentX=6.0057, extentY=2.4104, extentZ=8.6151, center=<8.8063 0.6439 -6.0424>, rotation=<rot 0 16.118 -82.6921>}]
}
constraints [
 {
  bone1Name=dyn_ribA
  bone2Name=dyn_ribA_Base
  swing1LimitDegrees=20
  swing2LimitDegrees=4
  swing2Motion=Locked
  twistLimitDegrees=4
  twistMotion=Locked
 }
 {
  bone1Name=dyn_ribB
  bone2Name=dyn_ribB_Base
  swing1LimitDegrees=20
  swing2LimitDegrees=4
  swing2Motion=Locked
  twistLimitDegrees=4
  twistMotion=Locked
 }
 {
  bone1Name=dyn_ribC
  bone2Name=dyn_ribC_Base
  swing1LimitDegrees=20
  swing2LimitDegrees=4
  swing2Motion=Locked
  twistLimitDegrees=4
  twistMotion=Locked
 }
 {
  bone1Name=dyn_ribD
  bone2Name=dyn_ribD_Base
  swing1LimitDegrees=20
  swing2LimitDegrees=4
  swing2Motion=Locked
  twistLimitDegrees=4
  twistMotion=Locked
 }
 {bone1Name=dyn_bellyJiggle, bone2Name=dyn_bellyJiggle_Base, swing1LimitDegrees=25, swing2LimitDegrees=25, twistLimitDegrees=12}
 {bone1Name=dyn_neckRibA, bone2Name=spine_05, swing1LimitDegrees=10, swing2LimitDegrees=4, twistLimitDegrees=8}
 {bone1Name=dyn_neckRibB, bone2Name=spine_05, swing1LimitDegrees=10, swing2LimitDegrees=4, twistLimitDegrees=8}
]
