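# Control Rig for the M_MED_Nana_Split player body that corrects right-side dyn_ skeletal bones (hip-to-knee, radius, ulna, fibula, tibia, scapula, clavicle).
# Each bone is aimed in global space by a from-two-vectors quaternion turning its X or negated X axis toward a target bone, with hip and knee targeting each other.
# Scapula and clavicle local Euler rotations are then clamped, and a weighted parent constraint blends dyn_pelvis_r between pelvis and dyn_hipToKnee_r.
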ControlRigBlueprint /Game/Characters/Player/Male/Medium/Bodies/M_MED_Nana_Split/Meshes/M_MED_Nana_Split_CtrlRig
graphs [
 RigVMFunctionLibrary {links [], nodes []}
 RigVMModel {nodes [
   n1 = &RigUnit_BeginExecution()
   n2 = &RigUnit_GetBoneTransform(Bone=dyn_hipToKnee_r)
   n3 = /Script/RigVM.RigVMFunction_MathQuaternionGetAxis(Axis=X, Quaternion=@n2.Transform.Rotation) {is=struct RigVMUnitNode, template="Axis::Execute(in Quaternion,in Axis,out Result)"}
   n4 = &RigUnit_GetBoneTransform(Bone=dyn_kneeToHip_r)
   n5 = /Script/RigVM.RigVMFunction_MathVectorSub(
    B=@n2.Transform.Translation
    A=/Script/RigVM.RigVMFunction_MathTransformTransformVector(Location=(X=0,Y=0,Z=0), Transform=@n4.Transform) {is=struct RigVMUnitNode}.Result
   ) {is=struct RigVMUnitNode, template="Subtract::Execute(in A,in B,out Result)"}
   /Script/ControlRig.RigUnit_SetBoneRotation(
    Bone=dyn_hipToKnee_r
    Space=GlobalSpace
    Weight=1
    bPropagateToChildren=True
    ExecutePin=@n1.ExecutePin
    Rotation=/Script/RigVM.RigVMFunction_MathQuaternionMul(
     B=@n2.Transform.Rotation
     A=/Script/RigVM.RigVMFunction_MathQuaternionFromTwoVectors(A=@n3.Result, B=@n5.Result) {is=struct RigVMUnitNode, template="FromTwoVectors::Execute(in A,in B,out Result)"}.Result
    ) {is=struct RigVMUnitNode, template="Multiply::Execute(in A,in B,out Result)"}.Result
   ) {is=struct RigVMUnitNode}
   n6 = &RigVMFunction_MathQuaternionGetAxis(Axis=X, Quaternion=@n4.Transform.Rotation) {template="Axis::Execute(in Quaternion,in Axis,out Result)"}
   n7 = &RigVMFunction_MathVectorSub(
    B=@n4.Transform.Translation
    A=&RigVMFunction_MathTransformTransformVector(Location=(X=0,Y=0,Z=0), Transform=@n2.Transform).Result
   ) {template="Subtract::Execute(in A,in B,out Result)"}
   > &RigUnit_SetBoneRotation(
    Bone=dyn_kneeToHip_r
    Space=GlobalSpace
    Weight=1
    bPropagateToChildren=True
    Rotation=&RigVMFunction_MathQuaternionMul(
     B=@n4.Transform.Rotation
     A=&RigVMFunction_MathQuaternionFromTwoVectors(A=@n6.Result, B=@n7.Result) {template="FromTwoVectors::Execute(in A,in B,out Result)"}.Result
    ) {template="Multiply::Execute(in A,in B,out Result)"}.Result
   )
   n8 = &RigUnit_GetBoneTransform(Bone=dyn_radius)
   n9 = /Script/RigVM.RigVMFunction_MathVectorNegate(Value=&RigVMFunction_MathQuaternionGetAxis(Axis=X, Quaternion=@n8.Transform.Rotation) {template="Axis::Execute(in Quaternion,in Axis,out Result)"}.Result) {is=struct RigVMUnitNode, template="Negate::Execute(in Value,out Result)"}
   n10 = &RigVMFunction_MathVectorSub(
    B=@n8.Transform.Translation
    A=&RigVMFunction_MathTransformTransformVector(Location=(X=0,Y=0,Z=0), Transform=&RigUnit_GetBoneTransform(Bone=dyn_radiusEnd).Transform).Result
   ) {template="Subtract::Execute(in A,in B,out Result)"}
   > &RigUnit_SetBoneRotation(
    Bone=dyn_radius
    Space=GlobalSpace
    Weight=1
    bPropagateToChildren=True
    Rotation=&RigVMFunction_MathQuaternionMul(
     B=@n8.Transform.Rotation
     A=&RigVMFunction_MathQuaternionFromTwoVectors(A=@n9.Result, B=@n10.Result) {template="FromTwoVectors::Execute(in A,in B,out Result)"}.Result
    ) {template="Multiply::Execute(in A,in B,out Result)"}.Result
   )
   n11 = &RigUnit_GetBoneTransform(Bone=dyn_ulna)
   n12 = &RigVMFunction_MathVectorNegate(Value=&RigVMFunction_MathQuaternionGetAxis(Axis=X, Quaternion=@n11.Transform.Rotation) {template="Axis::Execute(in Quaternion,in Axis,out Result)"}.Result) {template="Negate::Execute(in Value,out Result)"}
   n13 = &RigVMFunction_MathVectorSub(
    B=@n11.Transform.Translation
    A=&RigVMFunction_MathTransformTransformVector(Location=(X=0,Y=0,Z=0), Transform=&RigUnit_GetBoneTransform(Bone=dyn_ulnaEnd).Transform).Result
   ) {template="Subtract::Execute(in A,in B,out Result)"}
   > &RigUnit_SetBoneRotation(
    Bone=dyn_ulna
    Space=GlobalSpace
    Weight=1
    bPropagateToChildren=True
    Rotation=&RigVMFunction_MathQuaternionMul(
     B=@n11.Transform.Rotation
     A=&RigVMFunction_MathQuaternionFromTwoVectors(A=@n12.Result, B=@n13.Result) {template="FromTwoVectors::Execute(in A,in B,out Result)"}.Result
    ) {template="Multiply::Execute(in A,in B,out Result)"}.Result
   )
   n14 = &RigUnit_GetBoneTransform(Bone=dyn_fibula_Base)
   n15 = &RigVMFunction_MathQuaternionGetAxis(Axis=X, Quaternion=@n14.Transform.Rotation) {template="Axis::Execute(in Quaternion,in Axis,out Result)"}
   n16 = &RigVMFunction_MathVectorSub(
    B=@n14.Transform.Translation
    A=&RigVMFunction_MathTransformTransformVector(Location=(X=0,Y=0,Z=0), Transform=&RigUnit_GetBoneTransform(Bone=dyn_fibulaEnd).Transform).Result
   ) {template="Subtract::Execute(in A,in B,out Result)"}
   > &RigUnit_SetBoneRotation(
    Bone=dyn_fibula_Base
    Space=GlobalSpace
    Weight=1
    bPropagateToChildren=True
    Rotation=&RigVMFunction_MathQuaternionMul(
     B=@n14.Transform.Rotation
     A=&RigVMFunction_MathQuaternionFromTwoVectors(A=@n15.Result, B=@n16.Result) {template="FromTwoVectors::Execute(in A,in B,out Result)"}.Result
    ) {template="Multiply::Execute(in A,in B,out Result)"}.Result
   )
   n17 = &RigUnit_GetBoneTransform(Bone=dyn_tibia_Base)
   n18 = &RigVMFunction_MathQuaternionGetAxis(Axis=X, Quaternion=@n17.Transform.Rotation) {template="Axis::Execute(in Quaternion,in Axis,out Result)"}
   n19 = &RigVMFunction_MathVectorSub(
    B=@n17.Transform.Translation
    A=&RigVMFunction_MathTransformTransformVector(Location=(X=0,Y=0,Z=0), Transform=&RigUnit_GetBoneTransform(Bone=dyn_tibiaEnd).Transform).Result
   ) {template="Subtract::Execute(in A,in B,out Result)"}
   > &RigUnit_SetBoneRotation(
    Bone=dyn_tibia_Base
    Space=GlobalSpace
    Weight=1
    bPropagateToChildren=True
    Rotation=&RigVMFunction_MathQuaternionMul(
     B=@n17.Transform.Rotation
     A=&RigVMFunction_MathQuaternionFromTwoVectors(A=@n18.Result, B=@n19.Result) {template="FromTwoVectors::Execute(in A,in B,out Result)"}.Result
    ) {template="Multiply::Execute(in A,in B,out Result)"}.Result
   )
   n20 = &RigUnit_GetBoneTransform(Bone=dyn_scap)
   n21 = &RigVMFunction_MathQuaternionGetAxis(Axis=X, Quaternion=@n20.Transform.Rotation) {template="Axis::Execute(in Quaternion,in Axis,out Result)"}
   n22 = &RigVMFunction_MathTransformTransformVector(Location=(X=0,Y=0,Z=0), Transform=&RigUnit_GetBoneTransform(Bone=dyn_scapEnd).Transform)
   n23 = &RigVMFunction_MathQuaternionMul(
    B=@n20.Transform.Rotation
    A=&RigVMFunction_MathQuaternionFromTwoVectors(
     A=&RigVMFunction_MathVectorNegate(Value=@n21.Result) {template="Negate::Execute(in Value,out Result)"}.Result
     B=&RigVMFunction_MathVectorSub(B=@n20.Transform.Translation, A=@n22.Result) {template="Subtract::Execute(in A,in B,out Result)"}.Result
    ) {template="FromTwoVectors::Execute(in A,in B,out Result)"}.Result
   ) {template="Multiply::Execute(in A,in B,out Result)"}
   > &RigUnit_SetBoneRotation(Bone=dyn_scap, Space=GlobalSpace, Weight=1, bPropagateToChildren=True, Rotation=@n23.Result)
   n24 = &RigUnit_GetBoneTransform(Bone=dyn_clav)
   n25 = &RigVMFunction_MathVectorNegate(Value=&RigVMFunction_MathQuaternionGetAxis(Axis=X, Quaternion=@n24.Transform.Rotation) {template="Axis::Execute(in Quaternion,in Axis,out Result)"}.Result) {template="Negate::Execute(in Value,out Result)"}
   n26 = &RigVMFunction_MathVectorSub(
    B=@n24.Transform.Translation
    A=&RigVMFunction_MathTransformTransformVector(Location=(X=0,Y=0,Z=0), Transform=&RigUnit_GetBoneTransform(Bone=dyn_clavEnd).Transform).Result
   ) {template="Subtract::Execute(in A,in B,out Result)"}
   > &RigUnit_SetBoneRotation(
    Bone=dyn_clav
    Space=GlobalSpace
    Weight=1
    bPropagateToChildren=True
    Rotation=&RigVMFunction_MathQuaternionMul(
     B=@n24.Transform.Rotation
     A=&RigVMFunction_MathQuaternionFromTwoVectors(A=@n25.Result, B=@n26.Result) {template="FromTwoVectors::Execute(in A,in B,out Result)"}.Result
    ) {template="Multiply::Execute(in A,in B,out Result)"}.Result
   )
   n27 = &RigVMFunction_MathQuaternionToEuler(RotationOrder=ZYX, Value=&RigUnit_GetBoneTransform(Bone=dyn_scap, Space=LocalSpace).Transform.Rotation)
   > &RigUnit_SetBoneRotation(
    Bone=dyn_scap
    Space=LocalSpace
    Weight=1
    bPropagateToChildren=True
    Rotation=&RigVMFunction_MathQuaternionFromEuler(Euler=/Script/RigVM.RigVMFunction_MathVectorClamp(Maximum=(X=360,Y=360,Z=0), Minimum=(X=-360,Y=-360,Z=0), Value=@n27.Result) {is=struct RigVMUnitNode, template="Clamp::Execute(in Value,in Minimum,in Maximum,out Result)"}.Result).Result
   )
   n28 = &RigVMFunction_MathQuaternionToEuler(RotationOrder=ZYX, Value=&RigUnit_GetBoneTransform(Bone=dyn_clav, Space=LocalSpace).Transform.Rotation)
   > &RigUnit_SetBoneRotation(
    Bone=dyn_clav
    Space=LocalSpace
    Weight=1
    bPropagateToChildren=True
    Rotation=&RigVMFunction_MathQuaternionFromEuler(Euler=&RigVMFunction_MathVectorClamp(Maximum=(X=360,Y=360,Z=360), Minimum=(X=-360,Y=-360,Z=0), Value=@n28.Result) {template="Clamp::Execute(in Value,in Minimum,in Maximum,out Result)"}.Result).Result
   )
   > &RigUnit_ParentConstraint(
    AdvancedSettings=(InterpolationType=Average,RotationOrderForFilter=XZY)
    Child=(Type=Bone,Name="dyn_pelvis_r")
    Parents=((Item=(Type=Bone,Name="dyn_hipToKnee_r"),Weight=0.325),(Item=(Type=Bone,Name="pelvis"),Weight=0.675))
    Weight=1
   )
   &RigVMFunction_MathQuaternionFromEuler(Euler=/Script/RigVM.RigVMFunction_MathQuaternionRotateVector(Vector=(X=0,Y=0,Z=0), Transform=@n23.Result) {is=struct RigVMUnitNode, template="Rotate Vector::Execute(in Transform,in Vector,out Result)"}.Result)
   &RigVMFunction_MathVectorNegate(Value=@n18.Result) {template="Negate::Execute(in Value,out Result)"}
   &RigVMFunction_MathVectorNegate(Value=@n15.Result) {template="Negate::Execute(in Value,out Result)"}
   /Script/ControlRig.RigUnit_TransformConstraint(
    BaseBone=None
    BaseTransform=(Rotation=(X=0,Y=0,Z=0,W=1),Translation=(X=0,Y=0,Z=0),Scale3D=(X=1,Y=1,Z=1))
    BaseTransformSpace=GlobalSpace
    Bone=dyn_pelvis_r
    ExecutePin=()
    Targets="((Transform=(Rotation=(X=0,Y=0,Z=0,W=1),Translation=(X=0,Y=0,Z=0),Scale3D=(X=1,Y=1,Z=1)),Weight=0.325,bMaintainOffset=True,Filter=(TranslationFilter=(bX=True,bY=True,bZ=True),RotationFilter=(bX=True,bY=True,bZ=True),ScaleFilter=(bX=True,bY=True,bZ=True))), ...(truncated)"
    bUseInitialTransforms=True
    Targets.0.Transform=&RigUnit_GetBoneTransform(Bone=dyn_hipToKnee_r).Transform
    Targets.1.Transform=&RigUnit_GetBoneTransform(Bone=pelvis).Transform
   ) {is=struct RigVMUnitNode}
  ]}
]
hierarchy [
 root {loc=<0 0 0>, rot=<0 0 0 1>}
 attach {parent=root, loc=<0 0 0>, rot=<0 0 0 1>}
 pelvis {parent=root, loc=<0 -0.7071 93.1191>, rot=<0.0127 0.707 -0.0127 -0.707>}
 spine_01 {parent=pelvis, loc=<6.7494 -0.6388 0>, rot=<0 0 0.034 -0.9994>}
 spine_02 {parent=spine_01, loc=<7.2518 -0.035 0>, rot=<0 0 0.0238 -0.9997>}
 spine_03 {parent=spine_02, loc=<6.0104 0 0>, rot=<0 0 0.0149 -0.9999>}
 spine_04 {parent=spine_03, loc=<8.1024 0 0>, rot=<0 0 -0.0132 -0.9999>}
 spine_05 {parent=spine_04, loc=<10.0146 0 0>, rot=<0 0 -0.0015 -1>}
 clavicle_l {parent=spine_05, loc=<4.3575 -0.1466 -2.3223>, rot=<0.0465 -0.7289 -0.009 -0.6829>}
 upperarm_l {parent=clavicle_l, loc=<15.5111 0 0>, rot=<0.0305 -0.536 -0.0057 -0.8437>}
 lowerarm_l {parent=upperarm_l, loc=<26.6514 0 0>, rot=<0.0391 -0.0099 -0.2453 -0.9686>}
 hand_l {parent=lowerarm_l, loc=<22.3491 0 0>, rot=<-0.7081 -0.0507 0.0975 -0.6975>}
 index_metacarpal_l {parent=hand_l, loc=<1.6535 0.0183 -2.96>, rot=<0.0158 -0.0124 -0.0173 -0.9996>}
 index_01_l {parent=index_metacarpal_l, loc=<8.3008 0.0357 -0.0524>, rot=<-0.0044 0.0485 0.0924 -0.9945>}
 index_02_l {parent=index_01_l, loc=<3.3587 -0.3004 0.0452>, rot=<-0.002 0.0027 0.3225 -0.9466>}
 index_03_l {parent=index_02_l, loc=<2.5578 -0.123 -0.0252>, rot=<-0.002 0.0067 0.2558 -0.9667>}
 middle_metacarpal_l {parent=hand_l, loc=<2.1856 -0.1725 -0.5984>, rot=<0.006 0.0168 -0.0329 -0.9993>}
 middle_01_l {parent=middle_metacarpal_l, loc=<7.3275 -0.2447 0.0129>, rot=<-0.0208 0.0042 0.0858 -0.9961>}
 middle_02_l {parent=middle_01_l, loc=<4.0123 0.1401 0.1175>, rot=<-0.0218 0.0258 0.3512 -0.9357>}
 middle_03_l {parent=middle_02_l, loc=<2.6077 0.0534 0.1373>, rot=<0.0082 0.0188 0.1708 -0.9851>}
 pinky_metacarpal_l {parent=hand_l, loc=<1.7616 -0.9753 3.3814>, rot=<-0.0397 0.0814 -0.069 -0.9935>}
 pinky_01_l {parent=pinky_metacarpal_l, loc=<7.1709 -0.4831 -0.1094>, rot=<-0.0654 -0.0035 0.2242 -0.9723>}
 pinky_02_l {parent=pinky_01_l, loc=<2.8274 0.0893 -0.1043>, rot=<0.0037 0.0011 0.2854 -0.9584>}
 pinky_03_l {parent=pinky_02_l, loc=<2.2046 0.0307 -0.0078>, rot=<-0.0002 0.0098 0.1445 -0.9895>}
 ring_metacarpal_l {parent=hand_l, loc=<2.2074 -0.3625 1.3638>, rot=<-0.0284 0.0722 -0.0458 -0.9959>}
 ring_01_l {parent=ring_metacarpal_l, loc=<7.0295 -0.6197 -0.1446>, rot=<-0.0489 -0.0252 0.1892 -0.9804>}
 ring_02_l {parent=ring_01_l, loc=<3.6174 0.4647 -0.1027>, rot=<0 0 0.2806 -0.9598>}
 ring_03_l {parent=ring_02_l, loc=<2.4152 0.1141 0.0199>, rot=<0 0 0.1189 -0.9929>}
 thumb_01_l {parent=hand_l, loc=<1.3118 -0.3191 -2.459>, rot=<0.6546 -0.4042 0.0664 -0.6354>}
 thumb_02_l {parent=thumb_01_l, loc=<5.9336 -0.0448 0.0401>, rot=<-0.0037 0.0213 0.298 -0.9543>}
 thumb_03_l {parent=thumb_02_l, loc=<2.2223 0.3682 -0.0907>, rot=<0.0118 -0.013 0.0066 -0.9998>}
 weapon_l {parent=hand_l, loc=<-0.1974 -3.4111 0.3811>, rot=<0 0 0 -1>}
 lowerarm_twist_01_l {parent=lowerarm_l, loc=<17.8549 0 0>, rot=<0 0 0 1>}
 lowerarm_twist_02_l {parent=lowerarm_l, loc=<9.6199 0 0>, rot=<0 0 0 1>}
 upperarm_twist_01_l {parent=upperarm_l, loc=<0.0041 0 0>, rot=<0 0 0 1>}
 upperarm_twist_02_l {parent=upperarm_l, loc=<13.7941 0 0>, rot=<0 0 0 1>}
 clavicle_r {parent=spine_05, loc=<4.3575 -0.1466 2.3223>, rot=<0.7289 0.0465 -0.6829 0.009>}
 upperarm_r {parent=clavicle_r, loc=<-15.5111 0 0>, rot=<0.0305 -0.536 -0.0057 -0.8437>}
 lowerarm_r {parent=upperarm_r, loc=<-26.6514 0 0>, rot=<0.0391 -0.0099 -0.2453 -0.9686>}
 hand_r {parent=lowerarm_r, loc=<-22.3492 0 0>, rot=<-0.7081 -0.0507 0.0975 -0.6975>}
 index_metacarpal_r {parent=hand_r, loc=<-1.6534 -0.0181 2.9601>, rot=<0.0158 -0.0124 -0.0173 -0.9996>}
 index_01_r {parent=index_metacarpal_r, loc=<-8.3008 -0.0357 0.0524>, rot=<-0.0044 0.0485 0.0924 -0.9945>}
 index_02_r {parent=index_01_r, loc=<-3.3589 0.3003 -0.0453>, rot=<-0.002 0.0027 0.3225 -0.9466>}
 index_03_r {parent=index_02_r, loc=<-2.5578 0.1231 0.0251>, rot=<-0.002 0.0067 0.2558 -0.9667>}
 middle_metacarpal_r {parent=hand_r, loc=<-2.1856 0.1726 0.5984>, rot=<0.006 0.0168 -0.0329 -0.9993>}
 middle_01_r {parent=middle_metacarpal_r, loc=<-7.3273 0.2447 -0.0129>, rot=<-0.0208 0.0042 0.0858 -0.9961>}
 middle_02_r {parent=middle_01_r, loc=<-4.0142 -0.1411 -0.07>, rot=<-0.0217 0.0259 0.3512 -0.9357>}
 middle_03_r {parent=middle_02_r, loc=<-2.6074 -0.0534 -0.1312>, rot=<0.0083 0.0188 0.1708 -0.9851>}
 pinky_metacarpal_r {parent=hand_r, loc=<-1.7617 0.9753 -3.3814>, rot=<-0.0397 0.0814 -0.069 -0.9935>}
 pinky_01_r {parent=pinky_metacarpal_r, loc=<-7.1707 0.4832 0.1094>, rot=<-0.0654 -0.0035 0.2242 -0.9723>}
 pinky_02_r {parent=pinky_01_r, loc=<-2.8274 -0.0892 0.1043>, rot=<0.0037 0.0011 0.2854 -0.9584>}
 pinky_03_r {parent=pinky_02_r, loc=<-2.2045 -0.031 0.0078>, rot=<-0.0002 0.0098 0.1445 -0.9895>}
 ring_metacarpal_r {parent=hand_r, loc=<-2.2073 0.3627 -1.3638>, rot=<-0.0284 0.0722 -0.0458 -0.9959>}
 ring_01_r {parent=ring_metacarpal_r, loc=<-7.0297 0.6197 0.1446>, rot=<-0.0489 -0.0252 0.1892 -0.9804>}
 ring_02_r {parent=ring_01_r, loc=<-3.6171 -0.4648 0.1027>, rot=<0 0 0.2806 -0.9598>}
 ring_03_r {parent=ring_02_r, loc=<-2.4153 -0.1141 -0.0199>, rot=<0 0 0.1189 -0.9929>}
 thumb_01_r {parent=hand_r, loc=<-1.3119 0.3192 2.459>, rot=<0.6546 -0.4042 0.0664 -0.6354>}
 thumb_02_r {parent=thumb_01_r, loc=<-5.9335 0.0447 -0.0401>, rot=<-0.0037 0.0213 0.298 -0.9543>}
 thumb_03_r {parent=thumb_02_r, loc=<-2.2223 -0.3682 0.0907>, rot=<0.0118 -0.013 0.0066 -0.9998>}
 weapon_r {parent=hand_r, loc=<0.1975 3.4112 -0.3811>, rot=<0 0 0 -1>}
 dyn_radiusEnd {parent=hand_r, loc=<-0.444 -0.7213 0.6694>, rot=<-0.4834 -0.3223 -0.5887 -0.5621>}
 dyn_ulnaEnd {parent=hand_r, loc=<-0.2587 0.367 -1.8882>, rot=<-0.4834 -0.3223 -0.5887 -0.5621>}
 lowerarm_twist_01_r {parent=lowerarm_r, loc=<-17.8549 0 0>, rot=<0 0 0 1>}
 lowerarm_twist_02_r {parent=lowerarm_r, loc=<-9.6199 0 0>, rot=<0 0 0 1>}
 dyn_radius {parent=lowerarm_r, loc=<-0.2619 0.5091 -1.5222>, rot=<0.0224 -0.0193 -0.0253 -0.9992>}
 dyn_ulna {parent=lowerarm_r, loc=<1.3299 0.597 2.4238>, rot=<0.156 0.0371 0.0334 -0.9865>}
 upperarm_twist_01_r {parent=upperarm_r, loc=<-0.0041 0 0>, rot=<0 0 0 1>}
 upperarm_twist_02_r {parent=upperarm_r, loc=<-13.7942 0 0>, rot=<0 0 0 1>}
 dyn_elbowHalf {parent=upperarm_r, loc=<-26.6514 0 0>, rot=<0.0391 -0.0099 -0.2453 -0.9686>}
 dyn_scapEnd_Base {parent=clavicle_r, loc=<-14.316 1.6474 -3.012>, rot=<-0.0423 0.0083 0.9183 -0.3934>}
 dyn_scapEnd {parent=dyn_scapEnd_Base, loc=<0 0 0>, rot=<0 0 0 -1>}
 dyn_clavEnd_Base {parent=clavicle_r, loc=<-15.1833 -1.7381 -3.856>, rot=<-0.1122 0.0167 0.9398 0.3224>}
 dyn_clavEnd {parent=dyn_clavEnd_Base, loc=<0 0 0>, rot=<0 0 0 -1>}
 neck_01 {parent=spine_05, loc=<11.3138 0.1529 0>, rot=<0 0 -0.2003 -0.9797>}
 neck_02 {parent=neck_01, loc=<4.5627 0 0>, rot=<0 0 0.0504 -0.9987>}
 head {parent=neck_02, loc=<5.3914 0 0>, rot=<0 0 0.115 -0.9934>}
 dyn_skullRibD_Base {parent=neck_01, loc=<31.4569 -12.5512 2.944>, rot=<-0.6418 -0.6846 0.1441 0.314>}
 dyn_skullRibD {parent=dyn_skullRibD_Base, loc=<0 0 0>, rot=<0 0 0 1>}
 dyn_skullRibC_Base {parent=neck_01, loc=<37.2682 -11.5851 2.0052>, rot=<-0.6313 -0.7505 0.0717 0.1815>}
 dyn_skullRibC {parent=dyn_skullRibC_Base, loc=<0 0 0>, rot=<0 0 0 1>}
 dyn_skullRibB_Base {parent=neck_01, loc=<42.5466 -9.8833 1.4968>, rot=<-0.5695 -0.8041 0.0494 0.1635>}
 dyn_skullRibB {parent=dyn_skullRibB_Base, loc=<0 0 0>, rot=<0 0 0 1>}
 dyn_skullRibA_Base {parent=neck_01, loc=<46.6529 -7.5507 1.2335>, rot=<-0.5286 -0.8379 0.0322 0.1319>}
 dyn_skullRibA {parent=dyn_skullRibA_Base, loc=<0 0 0>, rot=<0 0 0 1>}
 dyn_headPeel_Base {parent=neck_01, loc=<34.6746 -5.7505 -8.4185>, rot=<-0.7413 -0.2088 -0.0105 -0.6378>}
 dyn_headPeel {parent=dyn_headPeel_Base, loc=<0 0 0>, rot=<0 0 0 -1>}
 dyn_jaw_Base {parent=neck_01, loc=<14.2023 0.5033 0>, rot=<-0.1086 -0.6987 0.1086 -0.6987>}
 dyn_jaw {parent=dyn_jaw_Base, loc=<0 0 0>, rot=<0 0 0 1>}
 dyn_neckRibA {parent=spine_05, loc=<18.7376 -10.918 2.6195>, rot=<-0.0815 -0.0322 -0.844 -0.5292>}
 dyn_neckRibB {parent=spine_05, loc=<12.8583 -11.8247 2.9145>, rot=<-0.0643 -0.0578 -0.814 -0.5745>}
 dyn_clav_Base {parent=spine_05, loc=<2.3385 14.536 1.6274>, rot=<-0.5933 -0.2891 0.7257 0.1944>}
 dyn_clav {parent=dyn_clav_Base, loc=<0 0 0>, rot=<0 0 0 1>}
 dyn_scap_Base {parent=spine_05, loc=<7.1412 -14.9934 3.36>, rot=<0.6587 -0.2388 -0.6541 0.285>}
 dyn_scap {parent=dyn_scap_Base, loc=<0 0 0>, rot=<0 0 0 -1>}
 dyn_ribA_Base {parent=spine_04, loc=<12.7062 0.364 3.4276>, rot=<0.3743 -0.5289 0.6562 0.3868>}
 dyn_ribA {parent=dyn_ribA_Base, loc=<0 0 0>, rot=<0 0 0 1>}
 dyn_ribB_Base {parent=spine_04, loc=<7.8526 0.1887 2.0666>, rot=<0.3844 -0.5319 0.6462 0.3896>}
 dyn_ribB {parent=dyn_ribB_Base, loc=<0 0 0>, rot=<0 0 0 1>}
 dyn_ribC_Base {parent=spine_04, loc=<1.984 -0.4674 1.7195>, rot=<0.4015 -0.5202 0.6329 0.4094>}
 dyn_ribC {parent=dyn_ribC_Base, loc=<0 0 0>, rot=<0 0 0 1>}
 dyn_ribD_Base {parent=spine_04, loc=<-2.8102 -0.6033 0.4655>, rot=<0.4148 -0.5599 0.6099 0.3776>}
 dyn_ribD {parent=dyn_ribD_Base, loc=<0 0 0>, rot=<0 0 0 1>}
 thigh_l {parent=pelvis, loc=<-2.5693 -0.0109 -7.7501>, rot=<0.0012 0.0572 0.0155 -0.9982>}
 calf_l {parent=thigh_l, loc=<-39.0484 0 0>, rot=<0.0006 -0.0161 -0.0391 -0.9991>}
 calf_twist_01_l {parent=calf_l, loc=<-35.8131 -0.0967 0.0397>, rot=<0.0237 0.0095 -0.0035 -0.9997>}
 calf_twist_02_l {parent=calf_l, loc=<-18.8064 -0.0967 0.0397>, rot=<0.0237 0.0095 -0.0035 -0.9997>}
 foot_l {parent=calf_l, loc=<-42.561 0.1077 0.4292>, rot=<0.0418 -0.0455 0.0343 -0.9975>}
 ball_l {parent=foot_l, loc=<-6.1599 13.8264 0.6806>, rot=<0.0001 -0.0001 -0.7063 -0.7079>}
 thigh_twist_01_l {parent=thigh_l, loc=<-21.6303 0 0>, rot=<0.0013 0 0 -1>}
 thigh_twist_shift_l {parent=thigh_twist_01_l, loc=<0 0 0>, rot=<0 0 0 1>}
 thigh_r {parent=pelvis, loc=<-2.5691 -0.0109 7.7501>, rot=<0.0572 -0.0012 0.9982 0.0155>}
 calf_r {parent=thigh_r, loc=<39.0487 0 0>, rot=<0.0006 -0.0161 -0.0391 -0.9991>}
 calf_twist_01_r {parent=calf_r, loc=<35.8132 0.0967 -0.0397>, rot=<0.0237 0.0095 -0.0035 -0.9997>}
 calf_twist_02_r {parent=calf_r, loc=<18.8064 0.0967 -0.0397>, rot=<0.0237 0.0095 -0.0035 -0.9997>}
 foot_r {parent=calf_r, loc=<42.5611 -0.1077 -0.4292>, rot=<0.0418 -0.0455 0.0343 -0.9975>}
 ball_r {parent=foot_r, loc=<6.1598 -13.8264 -0.6806>, rot=<0.0001 -0.0001 -0.7063 -0.7079>}
 dyn_fibulaEnd {parent=foot_r, loc=<0.4034 0.9053 2.2626>, rot=<0.7082 0.0284 -0.7045 -0.0357>}
 dyn_tibiaEnd {parent=foot_r, loc=<-0.0978 0.6412 -1.22>, rot=<0.7082 0.0284 -0.7045 -0.0357>}
 dyn_ball_r {parent=foot_r, loc=<7.3636 -9.0115 -0.6814>, rot=<0.0001 -0.0001 -0.7063 -0.7079>}
 dyn_fibula_Base {parent=calf_r, loc=<1.1732 2.1633 2.1061>, rot=<0.0118 -0.005 0.0144 -0.9998>}
 dyn_fibula {parent=dyn_fibula_Base, loc=<0 0 0>, rot=<0 0 0 1>}
 dyn_tibia_Base {parent=calf_r, loc=<1.8244 2.2746 -0.9984>, rot=<-0.046 -0.0074 0.023 -0.9986>}
 dyn_tibia {parent=dyn_tibia_Base, loc=<0 0 0>, rot=<0 0 0 1>}
 dyn_kneeToHip_r {parent=calf_r, loc=<0 0 0>, rot=<-0.0027 -0.0013 0.9989 0.0462>}
 thigh_twist_01_r {parent=thigh_r, loc=<21.6303 0 0>, rot=<0.0013 0 0 -1>}
 dyn_kneeHalf {parent=thigh_r, loc=<39.0487 0 0>, rot=<0.0006 -0.0161 -0.0391 -0.9991>}
 dyn_hipToKnee_r {parent=pelvis, loc=<-9.692 -0.2474 9.4244>, rot=<0.0458 0.059 0.997 0.0205>}
 dyn_bellyJiggle_Base {parent=pelvis, loc=<-5.4061 5.0519 2.7029>, rot=<-0.1061 0.1099 -0.6862 -0.7112>}
 dyn_bellyJiggle {parent=dyn_bellyJiggle_Base, loc=<0 0 0>, rot=<0 0 0 -1>}
 dyn_pelvis_r {parent=pelvis, loc=<-9.692 -0.2474 9.4244>, rot=<0.0458 0.059 0.997 0.0205>}
 ik_foot_root {parent=root, loc=<0 0 0>, rot=<0 0 0 1>}
 ik_foot_l {parent=ik_foot_root, loc=<15.2608 -4.4176 9.4433>, rot=<-0.0357 -0.7045 -0.0284 0.7082>}
 ik_foot_r {parent=ik_foot_root, loc=<-15.2609 -4.4176 9.4431>, rot=<-0.7082 -0.0284 0.7045 -0.0357>}
 ik_hand_root {parent=root, loc=<0 0 0>, rot=<0 0 0 1>}
 ik_hand_gun {parent=ik_hand_root, loc=<-33.8029 3.8283 90.5694>, rot=<-0.4834 -0.3223 -0.5887 0.5621>}
 ik_hand_l {parent=ik_hand_gun, loc=<6.7067 65.8041 13.978>, rot=<-0.0992 -0.9733 -0.2068 0>}
 ik_hand_r {parent=ik_hand_gun, loc=<0 0 0>, rot=<0 0 0 -1>}
]
