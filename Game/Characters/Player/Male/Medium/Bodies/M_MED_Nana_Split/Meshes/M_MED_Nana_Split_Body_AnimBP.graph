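# Post-style body AnimBlueprint for the male medium Nana Split character, adding secondary physics and hand-effect hiding to a copied pose.
# Copies the attached parent's pose and curves, runs it through the character's Control Rig, then a root-space RigidBody simulation.
# The physics alpha derives from jumping or crouching state, with interpolation and a raised floor.
# A weapon tagged Animation.CharacterParts.HideFX.Hands sets gauntletOn, which scales lowerarm_r to zero.
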AnimBlueprint /Game/Characters/Player/Male/Medium/Bodies/M_MED_Nana_Split/Meshes/M_MED_Nana_Split_Body_AnimBP
target_skeleton=&/M_MED_Nana_Split_Body_Skeleton
graphs [
 AnimGraph {nodes [
   n1 = Control Rig(Source=Copy Pose From Mesh() {bCopyCurves=true, bUseAttachedParent=true}) {
    alphaBoolBlend {blendOption=Linear}
    controlRigAssetReference {blueprintRigClass=&/M_MED_Nana_Split_CtrlRig.M_MED_Nana_Split_CtrlRig_C}
   }
   n2 = RigidBody(ComponentPose=Local To Component(LocalPose=@n1), Alpha=Get jumpCrouchAlpha() -> (jumpCrouchAlpha:double)) {
    bClampLinearTranslationLimitToRefPose=true
    bOverrideWorldGravity=true
    overridePhysicsAsset=&/M_MED_Nana_Split_Body_Physics
    simulationSpace=BaseBoneSpace
    alphaBoolBlend {blendOption=Linear}
    alphaScaleBiasClamp {bInterpResult=true, bMapRange=true, interpSpeedIncreasing=5, outRange {min=0.4}}
    baseBoneRef {boneName=root}
    componentAppliedLinearAccClamp=<1500 1500 1500>
    componentLinearAccScale=<0.3 0.3 0.5>
    componentLinearVelScale=<0.4 0.4 0.3>
    overrideWorldGravity {z=-980}
   }
   n3 = "Transform (Modify) Bone - Bone: lowerarm_r"(ComponentPose=@n2, Alpha=Get gauntletOn() -> (gauntletOn:double), Translation="0,0,0", Rotation="0,0,0", Scale="0,0,0") {scaleMode=BMM_Replace, scaleSpace=BCS_BoneSpace, alphaBoolBlend {blendOption=Linear}, boneToModify {boneName=lowerarm_r}, scale=<0 0 0>}
   Output Pose(Result=Component To Local(ComponentPose=@n3)) {layerGroup=None}
  ]}
 EventGraph {nodes [
   n1 = Cast To FortPawn!(execute, Object=GetOwningActor()) -> (then, CastFailed, AsFort Pawn:Fort Pawn Object Reference) {is=title K2Node_DynamicCast}
   n2 = Get CurrentWeapon!(self:Fort Pawn Object Reference=@n1) -> (CurrentWeapon:Fort Weapon Object Reference)
   > Set gauntletOn(gauntletOn:double="To Float (Boolean)"(InBool=WeaponDataHasAnyMatchingGameplayTags!(
      self:Fort Weapon Object Reference=@n2
      TagContainer:Gameplay Tag Container Structure=(GameplayTags=((TagName="Animation.CharacterParts.HideFX.Hands")))
     ) -> (ReturnValue:bool))) -> (Output_Get:double)
   Event BlueprintUpdateAnimation()
   n3 = OR Boolean(A=Get bIsJumping(), B=Get bIsCrouching())
   > Set jumpCrouchAlpha(jumpCrouchAlpha:double=float - float(B="To Float (Boolean)"(InBool=@n3)) {is=title K2Node_CallFunction}) -> (Output_Get:double)
   TryGetPawnOwner()
   Event OnWeaponChanged()
  ]}
]
variables [
 gauntletOn {type=number}
 jumpCrouchAlpha {type=number}
]
